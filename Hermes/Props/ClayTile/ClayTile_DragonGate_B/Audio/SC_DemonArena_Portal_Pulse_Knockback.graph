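# Sound cue for the knockback pulse of the Demon Arena portal on the ClayTile Dragon Gate B prop.
# A Random node picks one of four pulse knockback wave variations with equal weighting.
# It plays through a dedicated portal knockback attenuation asset and is routed to the Fort_Gadgets sound class.
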SoundCue /Hermes/Props/ClayTile/ClayTile_DragonGate_B/Audio/SC_DemonArena_Portal_Pulse_Knockback
root=Random_0
nodes [
 {id=Random_0, children [WavePlayer_0,WavePlayer_1,WavePlayer_2,WavePlayer_3], properties {weights [1,1,1,1]}}
 {id=WavePlayer_0, properties {soundWaveAssetPtr=&/DemonArena_Portal_Pulse_Knockback_01}}
 {id=WavePlayer_1, properties {soundWaveAssetPtr=&/DemonArena_Portal_Pulse_Knockback_02}}
 {id=WavePlayer_2, properties {soundWaveAssetPtr=&/DemonArena_Portal_Pulse_Knockback_03}}
 {id=WavePlayer_3, properties {soundWaveAssetPtr=&/DemonArena_Portal_Pulse_Knockback_04}}
]
properties {
 attenuationSettings=&/SA_DemonArena_Portal_Knockback
 maxDistance=3500
 soundClassObject=&Fort_Gadgets
}
