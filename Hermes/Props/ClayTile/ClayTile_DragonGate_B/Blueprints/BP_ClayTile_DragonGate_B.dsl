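; BP_ClayTile_DragonGate_B is a clay tile dragon gate prop whose portal effect shows or hides with the PortalIsActive state.
; ChangePortalState sets VFX_Gate visibility, called from BeginPlay and OnRep_PortalIsActive only on the Remote (non-authority) branch.
; Actors spawned by the PersistentFXComponent are hidden whenever the portal is inactive.
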(event EventBeginPlay
  (Networking|SwitchHasAuthority
    (:Remote
      (CallFunction|ChangePortalState (Variables|Default|GetPortalIsActive)))))

(event OnFxActorSpawned(PersistentFXComponent)
  (params (actor :type Actor%20Object%20Reference) (comp :type Persistent%20FXComponent%20Object%20Reference))
  (Rendering|SetActorHiddenInGame actor (!= (Variables|Default|GetPortalIsActive) true)))

(fn ChangePortalState
  (params (Active :type Boolean))
  (Rendering|SetVisibility (Variables|Default|GetVFX_Gate) Active false))

(fn OnRep_PortalIsActive
  (Networking|SwitchHasAuthority
    (:Remote
      (CallFunction|ChangePortalState (Variables|Default|GetPortalIsActive)))))

(fn ConstructionScript
  (|CallParentFunction))
