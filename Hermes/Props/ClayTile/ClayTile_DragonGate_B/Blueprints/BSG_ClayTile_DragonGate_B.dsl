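; Dragon Gate B portal prop whose impulse sphere, while active, knocks back non-NPC player pawns and tosses pickups.
; Replicated portal state toggles overlaps on the server and FX on clients, with snow portals keeping system FX off.
; Calendar event GrillMoon_After sets the permanent face-hide flag, whose OnRep shrinks and drops the head mesh.
; Construction sets face visibility, scales the light by mesh scale, and applies an editor-only FX override.
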(event Custom|TestHideFace
  (Variables|Default|SetPermanently%20Hide%20Face true))

(event EventBeginPlay
  (bind _impulsecollision (Variables|Default|GetImpulseCollision))
  (bind _portalisactive (Variables|Default|GetPortalIsActive))
  (Networking|SwitchHasAuthority
    (:Authority
      (Variables|/Script/Engine.PrimitiveComponent|SetbGenerateOverlapEvents _portalisactive _impulsecollision)
      (Components|Sphere|SetSphereRadius _impulsecollision (ScalableFloat|GetValueatLevel (Variables|Default|GetImpulseCollisionRadius_Hf)) false)
      (MCPContext|InitialCalendarSyncCompleted
        (:OnSyncComplete
          (bind _biseventactive (CalendarEvents|GetCalendarEventInformation "GrillMoon_After"))
          (if _biseventactive
            (Variables|Default|SetPermanently%20Hide%20Face true)))))
    (:Remote
      (Default|ChangePortalState _portalisactive))))

(event OnFxActorSpawned(PersistentFXComponent)
  (params (actor :type Actor%20Object%20Reference) (comp :type Persistent%20FXComponent%20Object%20Reference))
  (if (Variables|Default|GetSnowPortal)
    (Rendering|SetActorHiddenInGame actor true) ; particles flowing out of the portal
    (else
      (Rendering|SetActorHiddenInGame actor (!= (Variables|Default|GetPortalIsActive) true)) ; particles flowing out of the portal
    )))

(event OnComponentBeginOverlap(ImpulseCollision)
  (params (OverlappedComponent :type Primitive%20Component%20Object%20Reference) (OtherActor :type Actor%20Object%20Reference) (OtherComp :type Primitive%20Component%20Object%20Reference) (OtherBodyIndex :type Integer) (bFromSweep :type Boolean) (SweepResult :type Hit%20Result%20Structure%20%28by%20ref%29))
  (Networking|SwitchHasAuthority
    (:Authority
      (bind _asfort_player_pawn_athena (Utilities|Casting|CastToFortPlayerPawnAthena OtherActor)
        (:then
          (if (not (GameplayTags|HasMatchingGameplayTag _asfort_player_pawn_athena "(TagName=\"Pawn.Athena.NPC\")"))
            (GameplayEffects|FortApplyGameplayEffectToSelf :self (Ability|GetFortAbilitySystemComponent _asfort_player_pawn_athena) :GameplayEffectClass (Variables|Default|GetKnockbackGE) :OptionalSourceObject self :OptionalEffectCauser self :Level 0.0)))
        (:CastFailed
          (bind _asfort_pickup (Utilities|Casting|CastToFortPickup OtherActor))
          (bind (_returnvalue_roll _returnvalue_pitch _returnvalue_yaw) (Math|Rotator|BreakRotator (Math|Rotator|FindLookatRotation (Transformation|GetActorLocation) (Transformation|GetActorLocation _asfort_pickup))))
          (Pickup|TossPickupinDirection _asfort_pickup (* (Math|Vector|MakeVector _returnvalue_roll _returnvalue_pitch (Math|Float|Absolute(Float) _returnvalue_yaw)) (ScalableFloat|GetValueatLevel (Variables|Default|GetPickupTossForce_Hf))) 0 false true))))))

(fn ChangePortalState
  (params (Active :type Boolean))
  (bind _ne_systemshare (Variables|Default|GetNE_SystemShare))
  (if (Variables|Default|GetSnowPortal)
    (Rendering|SetVisibility _ne_systemshare false false) ; snow portal - always turn these off by default
    (else
      (Rendering|SetVisibility _ne_systemshare Active false))))

(fn OnRep_PermanentlyHideFace
  (bind _sm_head (Variables|Default|GetSM_Head))
  (if (Variables|Default|GetPermanently%20Hide%20Face)
    (Transformation|SetWorldScale3D _sm_head "0.010000,0.010000,0.010000") ; make the face impossible to see
    (Transformation|SetRelativeLocation _sm_head (- (Variables|/Script/Engine.SceneComponent|GetRelativeLocation _sm_head) "0.000000,0.000000,200.000000") false true) ; teleport it out of view
  ))

(fn OnRep_PortalIsActive
  (bind _portalisactive (Variables|Default|GetPortalIsActive))
  (Networking|SwitchHasAuthority
    (:Authority
      (Variables|/Script/Engine.PrimitiveComponent|SetbGenerateOverlapEvents _portalisactive (Variables|Default|GetImpulseCollision)))
    (:Remote
      (Default|ChangePortalState _portalisactive))))

(fn ConstructionScript
  (bind _lightsourceradius (Variables|Default|GetLightSourceRadius))
  (bind _scalefactor (LocalVariables|GetScaleFactor))
  (bind _returnvalue (* _scalefactor _lightsourceradius))
  (|CallParentFunction)
  (Utilities|FlowControl|Sequence
    (:then_0
      ; Noraml Construct / Init
      (Rendering|SetVisibility (Variables|Default|GetSM_Head) (Variables|Default|GetFaceVisibility) false) ; Show/Hide the face in the portal
      (Utilities|FlowControl|Sequence
        (:then_0
          (LocalVariables|SetScaleFactor (Math|Float|Max(Float) (.x (Transformation|GetWorldScale (Variables|StaticMeshActor|GetStaticMeshComponent))) (.y (Transformation|GetWorldScale (Variables|StaticMeshActor|GetStaticMeshComponent))) (.z (Transformation|GetWorldScale (Variables|StaticMeshActor|GetStaticMeshComponent))))))
        (:then_1
          (Rendering|Components|Light|SetLightColor (Variables|Default|GetPointLight) (Variables|Default|GetLight%20Color) true)
          (Rendering|Lighting|SetAttenuationRadius (Variables|Default|GetPointLight) (* (Variables|Default|GetLight%20Radius) (LocalVariables|GetScaleFactor)))
          (Rendering|Lighting|SetSourceRadius (Variables|Default|GetPointLight) _returnvalue)
          (Rendering|Lighting|SetSoftSourceRadius (Variables|Default|GetPointLight) _returnvalue)
          (Rendering|Components|Light|SetIntensity (Variables|Default|GetPointLight) (Variables|Default|GetLight%20Intensity)))
        (:then_2
          (if (Variables|Default|GetSnowPortal)
            _))) ; 0 - get the world scale of the portal\n1 - set the scale of components
    )
    (:then_1
      ; Editor-Only VFX Visibility Override
      (if (Editor|InEditor)
        (Default|ChangePortalState (Variables|Default|GetPortalIsActiveInEditor))))) ; 0 - normal init work\n1 - editor-only override for VFX
)
