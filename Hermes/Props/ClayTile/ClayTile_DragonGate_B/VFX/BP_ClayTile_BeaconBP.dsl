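; Actor Blueprint for a beacon in the VFX folder of the ClayTile DragonGate B prop.
; On BeginPlay, the authority turns off bAlwaysRelevant, which is only needed at startup so the actor spawns regardless of player location.
; The overlap and tick events and the construction script are empty, so this graph holds no other behavior.
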(event EventBeginPlay
  (Networking|SwitchHasAuthority
    (:Authority
      (Variables|Replication|SetbAlwaysRelevant false) ; We need to be always relevant to start so we spawn no matter where the player is but we don't need it after that.
    )))

(event Collision|EventActorBeginOverlap
  (params (OtherActor :type Actor%20Object%20Reference)))

(event EventTick
  (params (DeltaSeconds :type Float%20%28single-precision%29)))

(fn ConstructionScript)
