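# Material function that smooths particle SubUV flipbook playback by warping between frames with a motion vector texture.
# Motion vector RG is sampled at current and next frame UVs, remapped to an inverted signed range, and scaled by Strength and a direction switch input.
# Current frame UVs shift by SubUV blend, next frame UVs shift oppositely by its inverse.
# Both warped samples are crossfaded and output as RGBA, RGB and alpha.
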MaterialFunction /Hermes/Props/ClayTile/ClayTile_DragonGate_B/VFX/MF_MOON_SubUVMotionVectors_V2
graph {nodes [
  n1 = ParticleSubUVProperties()
  n2 = ParticleSubUVProperties()
  n3 = FunctionInput(Preview=TextureObject() {samplerType=SAMPLERTYPE_LinearColor, texture=null}) {bUsePreviewValueAsDefault=true, inputName=MotionVector, inputType=FunctionInput_Texture2D, sortPriority=1}
  n4 = TextureSample(UVs=@n2.TextureCoordinate0, Tex=@n3) {samplerType=SAMPLERTYPE_LinearColor}
  n5 = Multiply(
   A=FunctionInput(Preview=Constant() {r=-0.01}) {bUsePreviewValueAsDefault=true, inputName=Strength, inputType=FunctionInput_Scalar, sortPriority=3}
   B=FunctionInput(Preview=Constant() {r=1}) {bUsePreviewValueAsDefault=true, inputName="DirectionSwitchRG(1.-1)", inputType=FunctionInput_Vector2, sortPriority=4}
  )
  n6 = Multiply(A=ConstantBiasScale(None=ComponentMask(None=@n4.RGB) {g=true, r=true}) {bias=-0.5, scale=-2}, B=Multiply(A=@n5, B=@n2.Blend))
  n7 = FunctionInput(Preview=TextureObject() {samplerType=SAMPLERTYPE_LinearColor, texture=null}) {bUsePreviewValueAsDefault=true, inputName=SubUVtexture, inputType=FunctionInput_Texture2D, sortPriority=2}
  n8 = TextureSample(UVs=@n2.TextureCoordinate1, Tex=@n3) {samplerType=SAMPLERTYPE_LinearColor}
  n9 = Multiply(A=ConstantBiasScale(None=ComponentMask(None=@n8.RGB) {g=true, r=true}) {bias=-0.5, scale=-2}, B=Multiply(A=@n5, B=OneMinus(None=@n2.Blend)))
  n10 = LinearInterpolate(
   A=TextureSample(UVs=Add(A=@n1.TextureCoordinate0, B=@n6), Tex=@n7) {samplerType=SAMPLERTYPE_LinearColor}.RGBA
   B=TextureSample(UVs=Subtract(A=@n1.TextureCoordinate1, B=@n9), Tex=@n7) {samplerType=SAMPLERTYPE_LinearColor}.RGBA
   Alpha=@n1.Blend
  )
  FunctionOutput(None=@n10) {outputName=Result_RGBA, sortPriority=3}
  FunctionOutput(None=ComponentMask(None=@n10) {b=true, g=true, r=true}) {outputName=Result_RGB, sortPriority=1}
  FunctionOutput(None=ComponentMask(None=@n10) {a=true}) {outputName=Result_A, sortPriority=2}
 ]}
