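# M_BG is an unlit, two-sided glowing pink material in the ClayTile DragonGate VFX set, apparently the gate's background layer.
# Sphere masks with texture-driven hardness combine into a powered, tinted ramp for emissive.
# Deep BumpOffset parallax, radial coordinates, panning and tilt build the effect UVs.
# A PreviousFrameSwitch feeds panning noise, masked radially, only into the previous frame to fake motion vectors, alongside a scaled temporal dither.
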Material /Hermes/Props/ClayTile/ClayTile_DragonGate_B/VFX/M_BG
graph {
 properties {bUsedWithNanite=true, pixelDepthOffsetMode=PDOM_Legacy, shadingModel=MSM_Unlit, twoSided=true, usedShadingModels=Unlit}
 nodes [
  n1 = TextureSampleParameter2D!(UVs=NamedRerouteUsage(), Level=ScalarParameter() {defaultValue=1.5, parameterName=Param_6}, Apply View MipBias) {
   mipValueMode=TMVM_MipLevel
   parameterName=Param
   samplerType=SAMPLERTYPE_LinearColor
   texture=&/Tile_Tech_02
  }
  n2 = LinearInterpolate(
   A=ScalarParameter() {defaultValue=0.6132, parameterName=Param_3}
   B=ScalarParameter() {defaultValue=-0.2691, parameterName=Param_11}
   Alpha=SquareRoot(None=@n1.G)
  )
  n3 = SphereMask(A=TextureCoordinate(), B=Constant() {r=0.5}, Radius=ScalarParameter() {defaultValue=0.5, parameterName=Param_7}, Hardness=@n2) {hardnessPercent=50}
  n4 = SphereMask(A=NamedRerouteUsage(), B=Constant() {r=0.5}, Radius=ScalarParameter() {defaultValue=1.1629, parameterName=Param_1}, Hardness=@n2) {hardnessPercent=50}
  n5 = SphereMask(
   A=NamedRerouteUsage()
   B=Constant() {r=0.5}
   Radius=ScalarParameter() {defaultValue=0.676, parameterName=Param_2}
   Hardness=ScalarParameter() {defaultValue=0.2452, parameterName=Param_4}
  ) {hardnessPercent=50}
  n6 = Multiply(A=Max(A=OneMinus(None=@n3), B=OneMinus(None=@n4)), B=Add(B=OneMinus(None=@n5)))
  n7 = Multiply(
   A=Power(Base=@n6, Exp=ScalarParameter() {defaultValue=2, parameterName=Param_8})
   B=ScalarParameter() {defaultValue=2, parameterName=Param_9}
  )
  n8 = VectorParameter() {parameterName=Param2, defaultValue=<rgba 1 0.1198 0.2666 1>}
  MaterialFunctionCall(Emissive=Add(A=Multiply(A=@n7, B=@n8.RGB), B=VectorParameter() {defaultValue=<rgba 0.026 0.0028 0.0123 1>}.RGB)) {materialFunction=&MF_FX_EmissiveAdjustments}
  Multiply(
   A=MaterialFunctionCall!(
    InRamp=@n6
    CoreIntensity=ScalarParameter() {defaultValue=0.3439, parameterName=OverAllIntensity}
    DetailBalence=ScalarParameter() {defaultValue=1.5, parameterName=Param_6}
    MaxOutPut=Constant() {r=10}
   ) {materialFunction=&/MF_MOON_InvertDistanceRamp}.Result
   B=@n8.RGB
  )
  Multiply(
   A=ScalarParameter() {defaultValue=25, parameterName=Param_5}
   B=MaterialFunctionCall() {materialFunction=&DitherTemporalAA_Fortnite}.Result
  )
  NamedRerouteDeclaration(None=BumpOffset(Coordinate=TextureCoordinate(), Height=ScalarParameter() {defaultValue=-12.5039}) {referencePlane=0}) {name=BumpUV, nodeColor=<rgba 0.6406 0 1 1>}
  n9 = Add(A=NamedRerouteUsage()) {constB=-0.5}
  n10 = VectorParameter() {parameterName=Main_UV, defaultValue=<rgba 3 2.101 -0.1307 0.248>}
  NamedRerouteDeclaration(None=MaterialFunctionCall!(
    ClampTexture
    UV=MaterialFunctionCall(Vector or UVs=@n9) {materialFunction=&VectorToRadialValue}.Radial Coordinates
    Tex
    InRGB=@n10.RGB
    InA=@n10.RGB
    Time
    TimeMult
   ) {materialFunction=&/MF_MOON_TexPanner_V2}.PannerUV) {name=MainUV, nodeColor=<rgba 0 0.4049 1 1>}
  n11 = MaterialFunctionCall(Float2=NamedRerouteUsage()) {materialFunction=&BreakOutFloat2Components}
  n12 = Multiply(A=@n11.R, B=ScalarParameter() {defaultValue=1, parameterName=UV_Tilt})
  NamedRerouteDeclaration(None=MaterialFunctionCall(X=@n11.R, Y=Add(A=@n11.G, B=@n12)) {materialFunction=&MakeFloat2}.Result) {name=TiltedUV, nodeColor=<rgba 0 0.3848 1 1>}
  n13 = TextureCoordinate()
  Normalize(VectorInput=Subtract(A=@n13) {constB=0.5})
  n14 = TextureCoordinate()
  n15 = Constant() {r=0.5}
  n16 = ScalarParameter() {defaultValue=0.2691, parameterName=Param_12}
  n17 = Panner(Coordinate=@n13) {speedX=1, speedY=0.1249}
  n18 = Multiply(
   A=ConstantBiasScale(None=TextureSample(UVs=@n17) {samplerType=SAMPLERTYPE_LinearColor, texture=&/T_FX_Tile_0138_Moon}.RGB) {bias=-0.5, scale=2}
   B=ScalarParameter() {defaultValue=16.5219, parameterName=Param_10}
  )
  PreviousFrameSwitch(Current Frame=Constant(), Previous Frame=Multiply(A=SphereMask(A=@n14, B=@n15, Radius=@n16) {hardnessPercent=0}, B=@n18))
  n19 = Constant() {r=0.5}
  SphereMask(A=TextureCoordinate(), B=@n19, Radius=@n19)
  VectorParameter() {parameterName=Param_1, defaultValue=<rgba 28.7196 0 0 0>}
  VertexInterpolator(VS=@n18)
 ]
}
