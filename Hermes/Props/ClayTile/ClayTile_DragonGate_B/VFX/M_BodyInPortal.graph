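# Additive unlit material for skeletal meshes with depth test disabled, apparently for a glowing body seen inside the ClayTile DragonGate portal.
# A Custom node intersects the camera ray with a disc set by position, normal and radius parameters, providing a radial portal mask.
# Panned tile textures drive a two-color tint, alongside vertex-color-masked eye texture, Fresnel, camera and depth fades, and noise-driven vertex offset math.
# It also contains an eye-shader refraction network.
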Material /Hermes/Props/ClayTile/ClayTile_DragonGate_B/VFX/M_BodyInPortal
graph {
 properties {
  bAutomaticallySetUsageInEditor=false
  bDisableDepthTest=true
  bUsedWithMeshDeformer=true
  bUsedWithSkeletalMesh=true
  blendMode=BLEND_Additive
  pixelDepthOffsetMode=PDOM_Legacy
  shadingModel=MSM_Unlit
  usedShadingModels=Unlit
 }
 nodes [
  n1 = TextureSample(UVs=TextureCoordinate()) {texture=/Game/Characters/Shared/FNEye/Textures/T_Eye_D}
  Add(B=@n1.RGB)
  n2 = ScalarParameter() {defaultValue=0.784, parameterName=Param_10}
  n3 = Multiply(A=@n2) {constB=-1}
  n4 = Frac(None=Multiply(A=Time()) {constB=10})
  Add(A=Multiply(A=LinearInterpolate(A=@n3, B=@n2, Alpha=@n4) {constA=-0.1, constB=0.1}))
  DynamicParameter() {defaultValue=<rgba 1 1 1 250>, paramNames [Time,Distort,Dissolve,WPO_Width]}
  DynamicParameter() {defaultValue=<rgba 1 1 1 250>, paramNames [Time,Distort,Dissolve,WPO_Width]}
  MaterialExpressionMultiply_0 = Multiply(A=ObjectRadius(), B=ScalarParameter() {defaultValue=1, parameterName=R})
  MaterialExpressionCameraPositionWS_1 = CameraPositionWS()
  MaterialExpressionCameraVectorWS_0 = CameraVectorWS()
  MaterialExpressionTransform_1 = Transform(None=VectorParameter() {parameterName=N, defaultValue=<rgba 0 -1 0 0>}.RGB) {transformSourceType=TRANSFORMSOURCE_Local}
  MaterialExpressionTransformPosition_1 = TransformPosition(None=VectorParameter() {parameterName=P, defaultValue=<rgba 0 215.9079 59.2906 0>}.RGB) {transformType=TRANSFORMPOSSOURCE_World}
  n5 = Custom!(
   Radius=@MaterialExpressionMultiply_0
   CameraPos=@MaterialExpressionCameraPositionWS_1
   CameraVector=@MaterialExpressionCameraVectorWS_0
   PlaneNormal=@MaterialExpressionTransform_1
   PlanePos=@MaterialExpressionTransformPosition_1
  ) {
   code="// Input parameters\r\nfloat3 planeNormal = PlaneNormal; // Direction of the plane\r\nfloat3 planePos = PlanePos; // Position of the plane\r\nfloat3 rayOrigin = CameraPos; // Starting point of the ray\r\nfloat3 rayDir = -CameraVector; // Direction of the ray\r\nfloat radius = Radius; // Radius of the plane\r\n\r\n// Calculate the denominator\r\nfloat denom = dot(planeNormal, rayDir);\r\n\r\n// Check if the ray is parallel to the plane\r\nbool hit = denom > 1e-6;\r\n\r\n// Calculate the t value\r\nfloat t = hit ? dot(planeNormal, planePos - rayOrigin) / denom : -1.0;\r\n\r\n// Calculate the intersection point\r\nfloat3 intersection = hit &&&& t >= 0 ? rayOrigin + t * rayDir : float3(0.0, 0.0, 0.0);\r\n\r\n// Calculate the distance from the intersection point to the center of the plane\r\nfloat distance = length(intersection - planePos);\r\n\r\n// Check if the intersection point is within the circular plane\r\nbool withinRadius = distance <= radius;\r\n\r\n// Final intersection check\r\nbool finalHit = hit &&&& t >= 0 &&&& withinRadius;\r\n\r\n// Output the result\r\nreturn finalHit ? distance : float3(0.0, 0.0, 0.0);\r\n// return normalize(t);\r\n"
   showCode=true
   inputs [
    {inputName=Radius, input {expression=&self:MaterialExpressionMultiply_0}}
    {inputName=CameraPos, input {expression=&self:MaterialExpressionCameraPositionWS_1}}
    {
     inputName=CameraVector
     input {expression=&self:MaterialExpressionCameraVectorWS_0}
    }
    {inputName=PlaneNormal, input {expression=&self:MaterialExpressionTransform_1}}
    {inputName=PlanePos, input {expression=&self:MaterialExpressionTransformPosition_1}}
   ]
  }
  Frac(None=@n5)
  n6 = TextureCoordinate()
  LinearInterpolate(
   A=@n6
   B=MaterialFunctionCall(XY=@n6) {materialFunction=&Swizzle}.YX
   Alpha=ScalarParameter() {defaultValue=1, parameterName=Main_UV_Swizzle}
  )
  n7 = Subtract(A=ComponentMask(None=MaterialFunctionCall() {materialFunction=&BoundingBoxBased_0-1_UVW}.RGB) {b=true, r=true}) {constB=0.5}
  MaterialFunctionCall(Vector or UVs=@n7) {materialFunction=&VectorToRadialValue}
  n8 = ScalarParameter() {defaultValue=3, parameterName=Param_11}
  n9 = TextureSampleParameter2D!(UVs=NamedRerouteUsage(), Bias=@n8, Apply View MipBias) {
   mipValueMode=TMVM_MipBias
   parameterName=Param
   samplerType=SAMPLERTYPE_LinearColor
   texture=&/Tile_Tech_02
  }
  n10 = TextureSampleParameter2D(UVs=NamedRerouteUsage(), Bias=@n8) {
   mipValueMode=TMVM_MipBias
   parameterName=Param2
   samplerType=SAMPLERTYPE_LinearColor
   texture=&Tile_Tech_02
  }
  Multiply(A=@n9.G, B=@n10.G)
  n11 = ObjectRadius()
  Multiply(A=@n11, B=ScalarParameter() {defaultValue=1, parameterName=Param_2})
  Multiply(A=Saturate(None=Ceil(None=@n5)), B=NamedRerouteUsage())
  Multiply(A=VertexNormalWS(), B=ScalarParameter() {defaultValue=5000, parameterName=Param_13})
  n12 = ScalarParameter() {defaultValue=1.3057, parameterName=Param_12}
  Multiply(
   A=LinearInterpolate(
    A=VectorParameter() {parameterName=Param_1, defaultValue=<rgba 0.0521 0.0052 0.0122 1>}.RGB
    B=VectorParameter() {parameterName=Param_2, defaultValue=<rgba 4 0.3333 0.7611 1>}.RGB
    Alpha=LinearInterpolate(A=@n9.G, B=@n10.G, Alpha=@n12)
   )
   B=NamedRerouteUsage()
  )
  n13 = Multiply(A=OneMinus(None=@n1.G), B=ScalarParameter() {defaultValue=2, parameterName=Eye})
  n14 = VertexColor()
  n15 = NamedRerouteUsage()
  n16 = NamedRerouteUsage()
  n17 = VertexInterpolator(VS=Multiply(A=Multiply(A=@n15, B=@n16), B=NamedRerouteUsage()))
  n18 = OneMinus(None=Divide(A=@n5, B=@n11))
  NamedRerouteDeclaration(None=Multiply(A=LinearInterpolate(B=@n13, Alpha=@n14.G) {constA=1}, B=Multiply(A=@n17.PS, B=@n18))) {name=AlphaValue, nodeColor=<rgba 0 0.8359 1 1>}
  n19 = MaterialFunctionCall() {materialFunction=&BoundingBoxBased_0-1_UVW}
  n20 = VectorParameter() {defaultValue=<rgba 0.096 0.368 0.416 1>}
  NamedRerouteDeclaration(None=Multiply(
    A=MaterialFunctionCall(Input=@n19.B, Input Low=@n20.R, Input High=@n20.R) {materialFunction=&RemapValue_0-1}.Result
    B=MaterialFunctionCall(Input=@n19.G, Input Low=@n20.B, Input High=@n20.B) {materialFunction=&RemapValue_0-1}.Result
   )) {name=BoxRamp, nodeColor=<rgba 0 0.8125 1 1>}
  n21 = VectorParameter() {parameterName=Main_UV, defaultValue=<rgba 0.2 0.2 0.184 0.1>}
  NamedRerouteDeclaration(None=MaterialFunctionCall!(ClampTexture, UV=@n7, Tex, InRGB=@n21.RGB, InA=@n21.RGB, Time, TimeMult) {materialFunction=&/MF_MOON_TexPanner_V2}.PannerUV) {name=MainUV, nodeColor=<rgba 0 0.4049 1 1>}
  n22 = NamedRerouteUsage()
  n23 = NamedRerouteUsage()
  n24 = Subtract(A=Multiply(A=Subtract(A=@n22, B=@n23), B=NamedRerouteUsage()), B=Multiply(A=NamedRerouteUsage(), B=NamedRerouteUsage())) {desc=t}
  NamedRerouteDeclaration(None=Multiply(A=Normalize(VectorInput=@n24)) {constB=-1}) {name=Eye Refraction Direction, nodeColor=<rgba 0 0.2882 1 1>}
  n25 = NamedRerouteUsage()
  n26 = DotProduct(A=CameraVectorWS(), B=NamedRerouteUsage())
  n27 = LinearInterpolate(Alpha=Multiply(A=@n26, B=@n26)) {constA=0.325}
  NamedRerouteDeclaration(None=Multiply(A=NamedRerouteUsage(), B=Divide(A=@n25, B=@n27) {desc=heightW / cosAlpha - scale distance to iris plane by facing ratio of camera and plane})) {name=Refraction Vector, nodeColor=<rgba 0.0781 1 0 1>}
  n28 = NamedRerouteUsage()
  n29 = Constant2Vector() {g=0.5, r=0.5}
  n30 = NamedRerouteUsage()
  n31 = NamedRerouteUsage()
  NamedRerouteDeclaration(None=Power(Base=SphereMask(A=@n28, B=@n29, Radius=@n30, Hardness=@n31) {attenuationRadius=0.365, hardnessPercent=50}) {constExponent=0.5}) {name=eye displacement mask, nodeColor=<rgba 0.875 1 0 1>}
  n32 = NamedRerouteUsage()
  n33 = NamedRerouteUsage()
  NamedRerouteDeclaration(None=LinearInterpolate(A=@n32, B=Add(A=@n32, B=@n33), Alpha=NamedRerouteUsage())) {name=EyeUVs, nodeColor=<rgba 0 1 0.6016 1>}
  NamedRerouteDeclaration(None=ScalarParameter() {defaultValue=0.078, group=Eye Shader, parameterName="Eye Cornea Radius (UV)", sortPriority=10}) {name=EyeCorneaRadius, nodeColor=<rgba 0 1 0.2266 1>}
  NamedRerouteDeclaration(None=PixelNormalWS()) {name=_Normal, nodeColor=<rgba 0 0 1 1>}
  n34 = Subtract(A=CameraPositionWS(), B=WorldPosition() {worldPositionShaderOffset=WPT_ExcludeAllShaderOffsets}.XYZ)
  n35 = Abs(None=DotProduct(A=VertexNormalWS(), B=Normalize(VectorInput=@n34)))
  n36 = Power(
   Base=LinearInterpolate(A=OneMinus(None=@n35), B=@n35, Alpha=ScalarParameter() {defaultValue=1, parameterName="Fresnel_Invert (0or1)"})
   Exp=ScalarParameter() {defaultValue=1, parameterName=Fresnel_Power}
  )
  n37 = ScalarParameter() {parameterName=Fresnel_Offset}
  NamedRerouteDeclaration(None=Multiply(A=Subtract(A=@n36, B=@n37), B=Add(A=@n37))) {name=Fresnel, nodeColor=<rgba 1 0 0.9613 1>}
  n38 = NamedRerouteUsage()
  n39 = NamedRerouteUsage()
  NamedRerouteDeclaration(None=Multiply(A=NamedRerouteUsage(), B=DotProduct(A=@n38, B=@n39) {desc=facing}) {desc=w}) {name=w, nodeColor=<rgba 1 0.8297 0 1>}
  n40 = MaterialFunctionCall() {materialFunction=&BoundingBoxBased_0-1_UVW}
  n41 = VectorParameter() {parameterName=UV1, defaultValue=<rgba 1.9616 1.6875 0.048 0.08>}
  NamedRerouteDeclaration(None=MaterialFunctionCall(UV=ComponentMask(None=@n40.RGB) {b=true, r=true}, InRGB=@n41.RGB, InA=@n41.RGB) {materialFunction=&MF_MOON_TexPanner_V2}.PannerUV) {name=UV_1, nodeColor=<rgba 0 0.4049 1 1>}
  n42 = Subtract(A=CameraPositionWS(), B=ObjectPositionWS())
  n43 = Length(None=@n42) {is=class MaterialExpressionLength}
  n44 = ObjectRadius()
  n45 = Normalize(VectorInput=@n42)
  n46 = Transform(None=Constant3Vector() {constant=<rgba 0 0 -1 0>}) {transformSourceType=TRANSFORMSOURCE_View}
  n47 = Subtract(A=Multiply(A=Divide(A=@n43, B=@n44), B=DotProduct(A=@n45, B=@n46)), B=ScalarParameter() {defaultValue=1.5, parameterName=Camfade_Offset})
  n48 = ScalarParameter() {defaultValue=1, parameterName=Camfade_Lemgth}
  NamedRerouteDeclaration(None=Saturate(None=Divide(A=@n47, B=@n48))) {name=CamFade, nodeColor=<rgba 1 0 0.9613 1>}
  n49 = VectorParameter() {parameterName=UV2, defaultValue=<rgba 0.636 0.9737 -0.04 0.136>}
  NamedRerouteDeclaration(None=MaterialFunctionCall(UV=ComponentMask(None=@n40.RGB) {g=true, r=true}, InRGB=@n49.RGB, InA=@n49.RGB) {materialFunction=&MF_MOON_TexPanner_V2}.PannerUV) {name=UV_2, nodeColor=<rgba 0 0.4049 1 1>}
  NamedRerouteDeclaration(None=LinearInterpolate(
    A=NamedRerouteUsage()
    B=NamedRerouteUsage()
    Alpha=ScalarParameter() {defaultValue=1, group=Eye Shader, parameterName=Eye Refraction Mix, sliderMax=1, sortPriority=27}
   )) {name="Eye UVs (normalized)", nodeColor=<rgba 0 1 0.5232 1>}
  NamedRerouteDeclaration(None=Constant() {r=0.365}) {name=IrisSize, nodeColor=<rgba 1 0 0.7969 1>}
  n50 = Multiply(A=NamedRerouteUsage(), B=NamedRerouteUsage())
  n51 = NamedRerouteUsage()
  n52 = Multiply(A=NamedRerouteUsage(), B=NamedRerouteUsage())
  n53 = Add(
   A=Subtract(A=Divide(A=@n50, B=@n51)) {constB=-0.5}
   B=Divide(A=Multiply(A=@n52) {constB=-1, desc=compensates for normal map not being 0-1}, B=@n51)
  )
  n54 = TextureSampleParameter2D!(UVs=NamedRerouteDeclaration(None=Saturate(None=@n53)) {name=Iris UVs, nodeColor=<rgba 0.2422 1 0 1>}, Apply View MipBias) {
   group=Eye Shader
   parameterName=Eye Iris Normal
   samplerType=SAMPLERTYPE_Normal
   sortPriority=20
   texture=/Game/Characters/Shared/FNEye/Textures/T_Eye_Sphere_N
  }
  NamedRerouteDeclaration(None=@n54.RGB) {name="eyeNormal (unflattened)", nodeColor=<rgba 0 1 0.1797 1>}
  NamedRerouteDeclaration(None=ScalarParameter() {defaultValue=0.2, group=Eye Shader, parameterName=Eye Cornea Mask Hardness, sortPriority=11}) {name=Eye Cornea Mask Hardness Reroute, nodeColor=<rgba 0.8047 0 1 1>}
  n55 = SceneDepth()
  n56 = PixelDepth()
  n57 = ScalarParameter() {defaultValue=1}
  n58 = ObjectRadius()
  n59 = Divide(
   A=Add(A=Subtract(A=@n55.R, B=@n56.R), B=Multiply(A=@n57, B=@n58))
   B=Multiply(A=ScalarParameter() {defaultValue=0.1, parameterName=Param_7}, B=@n58)
  )
  n60 = Divide(
   A=Distance(A=ObjectPositionWS(), B=CameraPositionWS())
   B=Multiply(A=ScalarParameter() {defaultValue=10, parameterName=Param_8}, B=ObjectRadius())
  )
  n61 = Multiply(A=OneMinus(None=Saturate(None=@n60)), B=ScalarParameter() {defaultValue=4, parameterName=Param_9})
  NamedRerouteDeclaration(None=ShadingPathSwitch(Default=Saturate(None=@n59), Forward=@n61, Mobile=@n61)) {name=DepthCheck, nodeColor=<rgba 1 0 0.6562 1>}
  n62 = TextureSample(UVs=NamedRerouteUsage()) {
   constCoordinate=1
   desc=UVs not scaled by center as the base UVs are a ortho projection and this depth needs to stay in that space
   samplerType=SAMPLERTYPE_LinearColor
   texture=/Game/Characters/Shared/FNEye/Textures/T_EyeMidPlaneDisplacement
  }
  n63 = ScalarParameter() {defaultValue=0.4, group=Eye Shader, parameterName=Eye Iris UV Radius, sliderMax=0.5}
  n64 = TextureSample(UVs=MaterialFunctionCall(X=Add(A=@n63) {constB=0.5}, Y=Constant() {r=0.5}) {materialFunction=&MakeFloat2}.Result) {
   desc="Sample Displacement map as above, sampled at the Iris Radius"
   samplerType=SAMPLERTYPE_LinearColor
   texture=&T_EyeMidPlaneDisplacement
  }
  n65 = Subtract(A=@n62.RGB, B=@n64.R)
  n66 = Constant()
  NamedRerouteDeclaration(None=Multiply(A=Max(A=@n65, B=@n66), B=ScalarParameter() {defaultValue=0.5, group=Eye Shader, parameterName=Eye Iris Depth Scale}) {desc=Iris Depth}) {name=height, nodeColor=<rgba 0 1 0.4375 1>}
  n67 = NamedRerouteUsage()
  n68 = NamedRerouteUsage()
  n69 = TextureCoordinate()
  n70 = NamedRerouteUsage()
  NamedRerouteDeclaration(None=SphereMask(A=Multiply(A=@n67, B=@n68), B=Multiply(A=@n69, B=@n68) {constB=2}, Radius=Multiply(A=@n70) {constB=0.5}) {attenuationRadius=0.1}) {name=LeftEyeMask, nodeColor=<rgba 1 0 0.5391 1>}
  n71 = VectorParameter() {
   desc="uvPosition (UV0) of the left Eye. Top left is 0,0\r\nSet it to the same as the right eye if the same texture is used for both"
   group=Eye Shader
   parameterName="Eye Left UV Position (UV0)"
   sortPriority=13
   defaultValue=<rgba 0.94 0.35 0 1>
  }
  NamedRerouteDeclaration(None=ComponentMask(None=@n71.RGB) {g=true, r=true}) {name=Left Eye UVPosition, nodeColor=<rgba 1 0.375 0 1>}
  n72 = ScalarParameter() {defaultValue=2, desc="Texture Width/Height (2048x1024 would be 2)", group=Eye Shader, parameterName=Eye Texture AspectRatio, sortPriority=1}
  NamedRerouteDeclaration(None=MaterialFunctionCall(X=Constant() {r=1}, Y=Divide(B=@n72)) {materialFunction=&MakeFloat2}.Result) {name=AspectRatio, nodeColor=<rgba 1 0.4688 0 1>}
  n73 = VectorParameter() {group=Eye Shader, parameterName="Eye Right UV Position (UV0)", sortPriority=12, defaultValue=<rgba 0.94 0.122 0 0>}
  NamedRerouteDeclaration(None=ComponentMask(None=@n73.RGB) {g=true, r=true}) {name=RightEyeUVPosition, nodeColor=<rgba 0.7812 1 0 1>}
  n74 = TextureCoordinate()
  n75 = NamedRerouteDeclaration(None=Subtract(A=NamedRerouteUsage(), B=NamedRerouteUsage())) {name=LeftEyeUVDelta, nodeColor=<rgba 0.9688 1 0 1>}
  NamedRerouteDeclaration(None=LinearInterpolate(A=@n74, B=Subtract(A=@n74, B=@n75), Alpha=NamedRerouteUsage())) {name=UVs, nodeColor=<rgba 0 1 0.2969 1>}
  NamedRerouteDeclaration(None=Transform(None=NamedRerouteUsage())) {name=EyeDirection, nodeColor=<rgba 1 0.3047 0 1>}
  n76 = NamedRerouteUsage()
  n77 = MaterialFunctionCall() {materialFunction=&TangentBasis}
  n78 = NamedRerouteUsage()
  n79 = DotProduct(A=@n77.X U, B=@n78)
  n80 = Normalize(VectorInput=Subtract(A=@n77.X U, B=Multiply(A=@n79, B=@n78)))
  n81 = CrossProduct(A=@n80, B=@n78)
  n82 = Multiply(
   A=Multiply(
    A=ScalarParameter() {defaultValue=0.02, group=Eye Shader, parameterName=Eye Refraction Mult, sortPriority=25}
    B=Constant2Vector() {g=1, r=-1}
   )
   B=MaterialFunctionCall(X=DotProduct(A=@n80, B=@n76), Y=DotProduct(A=@n81, B=@n76)) {materialFunction=&MakeFloat2}.Result
  )
  n83 = NamedRerouteUsage()
  NamedRerouteDeclaration(None=Add(A=Multiply(A=@n82, B=@n83), B=NamedRerouteUsage())) {name=Refracted UVs, nodeColor=<rgba 0.6406 1 0 1>}
  NamedRerouteDeclaration(None=CameraVectorWS()) {name=Camera Vector, nodeColor=<rgba 1 0 1 1>}
  NamedRerouteDeclaration(None=Divide(A=Constant() {desc=airIoR, r=1.0003}, B=ScalarParameter() {defaultValue=1.336, group=Eye Shader, parameterName=Eye Cornea IOR}) {constA=1.0003}) {name=n, nodeColor=<rgba 0 1 0.2031 1>}
  n84 = NamedRerouteUsage()
  n85 = NamedRerouteUsage()
  n86 = Multiply(A=Subtract(A=@n84, B=@n85), B=Add(A=@n84, B=@n85))
  NamedRerouteDeclaration(None=SquareRoot(None=Add(A=@n86)) {desc=k}) {name=k, nodeColor=<rgba 0.518 1 0 1>}
  NamedRerouteUsage()
  n87 = OneMinus(None=NamedRerouteUsage())
  n88 = ScalarParameter() {defaultValue=5, parameterName=Param_1}
  n89 = Multiply(A=OneMinus(None=NamedRerouteUsage()), B=Add(A=ScalarParameter() {defaultValue=1, parameterName=Param_4}, B=Power(Base=@n87, Exp=@n88)))
  n90 = Normalize(VectorInput=TransformPosition(None=WorldPosition().XYZ) {transformSourceType=TRANSFORMPOSSOURCE_World})
  n91 = CrossProduct(A=@n90, B=Constant3Vector() {constant=<rgba 0 -1 0 0>})
  n92 = LinearInterpolate(
   A=@n90
   B=Transform(None=Normalize(VectorInput=@n91)) {transformSourceType=TRANSFORMSOURCE_Local}
   Alpha=ScalarParameter() {defaultValue=0.5, parameterName=Param_3}
  )
  n93 = LinearInterpolate(
   A=ScalarParameter() {defaultValue=-0.17, parameterName=Param_5}
   B=ScalarParameter() {defaultValue=0.26, parameterName=Param_6}
   Alpha=TextureSample(UVs=NamedRerouteUsage()) {samplerType=SAMPLERTYPE_LinearColor, texture=&/T_FX_Tile_0138_Moon}.G
  ) {constA=0.5}
  n94 = ObjectRadius()
  PreviousFrameSwitch(Current Frame=Multiply(A=Multiply(A=@n89, B=@n92), B=Multiply(A=@n93, B=@n94)), Previous Frame=Constant())
  VertexColor()
 ]
}
