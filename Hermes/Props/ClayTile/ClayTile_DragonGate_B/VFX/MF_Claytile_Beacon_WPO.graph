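# Material function for the clay tile Dragon Gate beacon VFX that collapses the mesh into its actor position as the camera gets close.
# Per-vertex camera distance is offset by the Distance parameter, divided by Beam soften, saturated and inverted.
# That weight scales the vector from world position to actor position as WPO output, so geometry is fully collapsed within Distance and regains its shape over the soften range.
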MaterialFunction /Hermes/Props/ClayTile/ClayTile_DragonGate_B/VFX/MF_Claytile_Beacon_WPO
graph {nodes [
  n1 = Length(None=Subtract(A=CameraPositionWS(), B=WorldPosition().XYZ)) {is=class MaterialExpressionLength}
  n2 = ScalarParameter() {defaultValue=4500, parameterName=Distance}
  n3 = Saturate(None=Divide(A=Subtract(A=@n1, B=@n2), B=ScalarParameter() {defaultValue=4500, parameterName=Beam soften}))
  n4 = ActorPositionWS()
  n5 = WorldPosition()
  FunctionOutput(None=Multiply(A=OneMinus(None=@n3), B=Subtract(A=@n4, B=@n5.XYZ))) {outputName=WPO}
 ]}
