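# Translucent, unlit, two-sided material for the energy beam of the ClayTile Dragon Gate beacon.
# Panning noise, distorted by a swirl normal, warps UV-based sphere masks that shape the beam, multiplied by a camera-distance fade that hides it up close.
# A time-looped sphere mask forms a "Sweep" pulse along V.
# World position offset math combines SplineThicken, widened by V gradients, with MF_Claytile_Beacon_WPO, alongside a player-distance-weighted offset toward the actor position.
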Material /Hermes/Props/ClayTile/ClayTile_DragonGate_B/VFX/M_Claytile_Beacon_Energy
graph {
 properties {bAutomaticallySetUsageInEditor=false, blendMode=BLEND_Translucent, shadingModel=MSM_Unlit, twoSided=true, usedShadingModels=Unlit}
 nodes [
  n1 = TextureCoordinate()
  n2 = Multiply(A=Time(), B=ScalarParameter() {defaultValue=0.5, parameterName=Param_3})
  n3 = ScalarParameter() {defaultValue=2, parameterName=Pulse Frequency}
  n4 = SphereMask(
   A=OneMinus(None=ComponentMask(None=@n1) {g=true})
   B=Power(Base=Modulo!(A=@n2, B=@n3), Exp=ScalarParameter() {defaultValue=1.5, parameterName=Param_1})
   Radius=ScalarParameter() {defaultValue=0.01}
  ) {hardnessPercent=0}
  Add(A=Multiply(A=@n4, B=NamedRerouteUsage()))
  Add()
  n5 = ComponentMask(None=TextureCoordinate()) {g=true}
  n6 = ScalarParameter() {defaultValue=3, parameterName=Param_4}
  n7 = Add(
   A=LinearInterpolate(B=ScalarParameter() {parameterName=Param_2}, Alpha=NamedRerouteUsage())
   B=LinearInterpolate(B=ScalarParameter() {defaultValue=807.7798, parameterName=Width Base}, Alpha=Power(Base=@n5, Exp=@n6))
  )
  n8 = NamedRerouteUsage()
  Add(
   A=MaterialFunctionCall(WidthBase=Multiply(A=@n7, B=@n8), WidthTip=ScalarParameter() {defaultValue=350, parameterName=Width Tip}) {materialFunction=&SplineThicken}.WorldPositionOffset
   B=MaterialFunctionCall() {materialFunction=&/MF_Claytile_Beacon_WPO}.WPO
  )
  CameraPositionWS()
  Multiply()
  n9 = CollectionParameter() {collection=&FortniteMaterialParameters, parameterName=PlayerPosition}
  n10 = Subtract(A=ComponentMask(None=@n9) {b=true, g=true, r=true}, B=WorldPosition().XYZ)
  n11 = Divide(
   A=Subtract(A=Length(None=@n10) {is=class MaterialExpressionLength}, B=ScalarParameter() {defaultValue=10000, parameterName=Distance})
   B=ScalarParameter() {defaultValue=15000, parameterName=Beam soften}
  )
  Multiply(A=OneMinus(None=Saturate(None=@n11)), B=Subtract(A=ActorPositionWS(), B=WorldPosition().XYZ))
  n12 = Multiply(
   A=TextureCoordinate()
   B=ComponentMask(None=VectorParameter() {parameterName=Distort Tiling, defaultValue=<rgba 1 5 0 0>}.RGB) {g=true, r=true}
  )
  n13 = Multiply(A=Time(), B=ScalarParameter() {defaultValue=0.06, parameterName=PanSpeed})
  n14 = Time()
  n15 = ScalarParameter() {defaultValue=0.05, parameterName=Param_11}
  n16 = TextureSample(UVs=Panner(Coordinate=@n12, Time=Multiply(A=@n14, B=@n15)) {speedY=0.5}) {samplerType=SAMPLERTYPE_Normal, texture=/Game/Gadgets/Assets/VinderTech_GliderChute/Glider_Thrive_Female/FX/Texture/T_SwirlNoise_256}
  n17 = ComponentMask(None=@n16.RGB) {g=true, r=true}
  n18 = ScalarParameter() {defaultValue=0.05, parameterName=distort scale}
  n19 = TextureSample(UVs=Add(A=Panner(Coordinate=@n12, Time=@n13) {speedY=1}, B=Multiply(A=@n17, B=@n18))) {samplerType=SAMPLERTYPE_LinearColor, texture=&/T_FX_Tile_0138_Moon}
  n20 = ScalarParameter() {defaultValue=0.3, parameterName=Param_7}
  n21 = VectorParameter() {parameterName=Distort Offset, defaultValue=<rgba -0.1136 0 0 0>}
  n22 = Add(A=Add(A=Multiply(A=@n19.G, B=@n20), B=ComponentMask(None=@n21.RGB) {g=true, r=true}), B=TextureCoordinate())
  n23 = ScalarParameter() {defaultValue=-0.5887, parameterName=Param_15}
  n24 = ScalarParameter() {defaultValue=0.3792, parameterName=Param_6}
  n25 = Power(Base=ComponentMask(None=TextureCoordinate()) {g=true}, Exp=ScalarParameter() {defaultValue=2, parameterName=Param_16})
  n26 = OneMinus(None=SphereMask(
    A=ComponentMask(None=@n22) {r=true}
    B=ScalarParameter() {defaultValue=0.508, parameterName=GradientLocation}
    Radius=LinearInterpolate(A=@n23, B=@n24, Alpha=@n25)
    Hardness=ScalarParameter() {parameterName=Param_5}
   ) {attenuationRadius=0.5, hardnessPercent=0})
  n27 = SphereMask(
   A=ComponentMask(None=TextureCoordinate()) {r=true}
   B=ScalarParameter() {defaultValue=0.5, parameterName=Param_8}
   Radius=ScalarParameter() {defaultValue=0.5272, parameterName=Param_9}
   Hardness=ScalarParameter() {defaultValue=0.4701, parameterName=Param_10}
  ) {hardnessPercent=0}
  n28 = ComponentMask(None=TextureCoordinate()) {g=true}
  n29 = ScalarParameter() {defaultValue=1, parameterName=Param_12}
  n30 = ScalarParameter() {defaultValue=0.8105, parameterName=Param_13}
  n31 = ScalarParameter() {defaultValue=0.1498, parameterName=Param_14}
  n32 = Length(None=Subtract(A=CameraPositionWS(), B=WorldPosition().XYZ))
  n33 = ScalarParameter() {defaultValue=14500, parameterName=Distance fADE}
  n34 = Saturate(None=Divide(A=Subtract(A=@n32, B=@n33), B=ScalarParameter() {defaultValue=9500, parameterName=Beam soften fade}))
  Multiply(A=Multiply(A=Multiply(A=@n26, B=@n27), B=SphereMask(A=@n28, B=@n29, Radius=@n30, Hardness=@n31)), B=@n34)
  NamedRerouteDeclaration(None=@n4) {name=Sweep, nodeColor=<rgba 1 0 0.4453 1>}
  NamedRerouteDeclaration(None=@n34) {name=Fade, nodeColor=<rgba 1 0 0.6094 1>}
  OneMinus()
  PixelDepth()
  VectorParameter() {parameterName=Beam Color, defaultValue=<rgba 0 0 0 1>}
 ]
}
