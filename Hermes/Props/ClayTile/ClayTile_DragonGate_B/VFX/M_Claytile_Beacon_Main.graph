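# Translucent, unlit, two-sided material for the ClayTile Dragon Gate beacon, drawing a colored beam with a repeating sweep pulse.
# SplineThicken shapes the beam width along UV height, combined with an MF_Claytile_Beacon_WPO offset.
# A time-modulo SphereMask creates the sweep, and UV sphere masks form vertical and horizontal gradients.
# The gradients multiply by a camera-distance fade that is zero up close and ramps in beyond a threshold.
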Material /Hermes/Props/ClayTile/ClayTile_DragonGate_B/VFX/M_Claytile_Beacon_Main
graph {
 properties {bAutomaticallySetUsageInEditor=false, blendMode=BLEND_Translucent, shadingModel=MSM_Unlit, twoSided=true, usedShadingModels=Unlit}
 nodes [
  n1 = TextureCoordinate()
  n2 = Multiply(A=Time(), B=ScalarParameter() {defaultValue=0.5, parameterName=Param_3})
  n3 = ScalarParameter() {defaultValue=2, parameterName=Pulse Frequency}
  n4 = SphereMask(
   A=OneMinus(None=ComponentMask(None=@n1) {g=true})
   B=Power(Base=Modulo!(A=@n2, B=@n3), Exp=ScalarParameter() {defaultValue=1.5, parameterName=Param_1})
   Radius=ScalarParameter() {defaultValue=0.01}
  ) {hardnessPercent=0}
  Add(A=Multiply(A=@n4, B=NamedRerouteUsage()))
  n5 = NamedRerouteUsage()
  n6 = ComponentMask(None=TextureCoordinate()) {g=true}
  n7 = ScalarParameter() {defaultValue=4, parameterName=Param_4}
  n8 = Add(
   A=LinearInterpolate(B=ScalarParameter() {parameterName=Param_2}, Alpha=NamedRerouteUsage())
   B=LinearInterpolate(B=ScalarParameter() {defaultValue=512.0108, parameterName=Width Base}, Alpha=Power(Base=@n6, Exp=@n7))
  )
  Add(
   A=MaterialFunctionCall(WidthBase=Multiply(A=@n5, B=@n8), WidthTip=ScalarParameter() {parameterName=Width Tip}) {materialFunction=&SplineThicken}.WorldPositionOffset
   B=MaterialFunctionCall() {materialFunction=&/MF_Claytile_Beacon_WPO}.WPO
  )
  n9 = CollectionParameter() {collection=&FortniteMaterialParameters, parameterName=PlayerPosition}
  ComponentMask(None=@n9) {b=true, g=true, r=true}
  LinearInterpolate()
  n10 = Multiply(B=ScalarParameter() {defaultValue=15, parameterName=Param_8})
  Multiply(
   A=@n10
   B=MaterialFunctionCall(Alpha Threshold=@n10) {materialFunction=&DitherTemporalAA}.Result
  )
  Multiply() {constB=-1}
  n11 = TextureCoordinate()
  n12 = SphereMask(
   A=ComponentMask(None=@n11) {g=true}
   B=ScalarParameter() {defaultValue=1, parameterName=Param_5}
   Radius=ScalarParameter() {defaultValue=0.208, parameterName=Param_9}
   Hardness=ScalarParameter() {parameterName=Param_10}
  )
  n13 = SphereMask(A=ComponentMask(None=@n11) {r=true}, B=ScalarParameter() {defaultValue=0.5, parameterName=GradientLocation}) {attenuationRadius=0.5, hardnessPercent=0}
  n14 = Length(None=Subtract(A=CameraPositionWS(), B=WorldPosition().XYZ)) {is=class MaterialExpressionLength}
  n15 = ScalarParameter() {defaultValue=16500, parameterName=Distance fADE}
  n16 = Saturate(None=Divide(A=Subtract(A=@n14, B=@n15), B=ScalarParameter() {defaultValue=7500, parameterName=Beam soften fade}))
  Multiply(A=Multiply(A=Multiply(A=@n12, B=@n13), B=ScalarParameter() {defaultValue=2, parameterName=Param_6}), B=@n16)
  NamedRerouteDeclaration(None=@n12) {name=Verrical Gradient, nodeColor=<rgba 0.1953 0 1 1>}
  NamedRerouteDeclaration(None=@n4) {name=Sweep, nodeColor=<rgba 1 0 0.4453 1>}
  NamedRerouteDeclaration(None=@n16) {name=Fade, nodeColor=<rgba 0 0.0625 1 1>}
  ObjectPositionWS()
  PixelDepth()
  ScalarParameter() {parameterName=Param_7}
  SceneDepth()
  TransformPosition(None=VectorParameter() {defaultValue=<rgba 0 0 0 1>}.RGB) {transformSourceType=TRANSFORMPOSSOURCE_World}
  TransformPosition() {transformSourceType=TRANSFORMPOSSOURCE_World}
  VectorParameter() {parameterName=Beam Color, defaultValue=<rgba 10 0 0.9295 1>}
 ]
}
