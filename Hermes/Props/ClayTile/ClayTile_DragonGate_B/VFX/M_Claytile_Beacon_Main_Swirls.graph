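# Translucent, unlit, two-sided material for animated smoke swirls on the main beacon effect of the ClayTile Dragon Gate B prop.
# A panning, tiled smoke-trail texture both displaces vertices along normals and forms a mask, shaped by a V-axis sphere mask and faded by camera distance.
# The vertex-offset chain adds SplineThicken, time-driven rotation around the pivot's Z axis and the shared MF_Claytile_Beacon_WPO function to that displacement.
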Material /Hermes/Props/ClayTile/ClayTile_DragonGate_B/VFX/M_Claytile_Beacon_Main_Swirls
graph {
 properties {bAutomaticallySetUsageInEditor=false, blendMode=BLEND_Translucent, shadingModel=MSM_Unlit, twoSided=true, usedShadingModels=Unlit}
 nodes [
  n1 = PixelDepth()
  n2 = ScalarParameter() {defaultValue=1000, parameterName=Distance}
  n3 = Saturate(None=Divide(A=Subtract(A=@n1.R, B=@n2), B=ScalarParameter() {defaultValue=2500, parameterName=Beam soften}))
  n4 = ActorPositionWS()
  n5 = WorldPosition()
  Add(A=Multiply(A=OneMinus(None=@n3), B=Subtract(A=@n4, B=@n5.XYZ)))
  n6 = Multiply(A=TextureCoordinate(), B=ComponentMask(None=VectorParameter() {parameterName=Tiling, defaultValue=<rgba 1 7 0 0>}.RGB) {g=true, r=true})
  n7 = Multiply(A=Time(), B=ScalarParameter() {defaultValue=0.1, parameterName=TimeScale})
  n8 = TextureSample(UVs=MaterialFunctionCall(XY=Panner(Coordinate=@n6, Time=@n7) {speedY=1}) {materialFunction=&Swizzle}.YX) {samplerType=SAMPLERTYPE_Grayscale, texture=/Game/Effects/Fort_Effects/Textures/Smoke/T_Smoke_Trail_Soft_NoClamp}
  n9 = Multiply(A=@n8.G, B=VertexNormalWS())
  n10 = ScalarParameter() {defaultValue=50000, parameterName=Param_6}
  n11 = MaterialFunctionCall(WidthBase=ScalarParameter() {defaultValue=950, parameterName=Param_1}, WidthTip=ScalarParameter() {defaultValue=250}) {materialFunction=&SplineThicken}
  n12 = RotateAboutAxis(
   NormalizedRotationAxis=Constant3Vector() {constant=<rgba 0 0 1 0>}
   RotationAngle=Multiply(A=ScalarParameter() {defaultValue=0.05, parameterName=Param_5}, B=Time())
   PivotPoint=MaterialFunctionCall() {materialFunction=&ObjectPivotPoint}.Object Pivot Location
   Position=WorldPosition().XYZ
  )
  Add(
   A=Add(A=Multiply(A=@n9, B=@n10), B=Add(A=@n11.WorldPositionOffset, B=@n12))
   B=MaterialFunctionCall() {materialFunction=&/MF_Claytile_Beacon_WPO}.WPO
  )
  n13 = ComponentMask(None=TextureCoordinate()) {r=true}
  Multiply(
   A=Multiply(A=Subtract(A=@n13) {constB=0.5}, B=MaterialFunctionCall() {materialFunction=&TangentBasis}.X U)
   B=ScalarParameter() {defaultValue=1500, parameterName=Param_4}
  )
  n14 = CameraPositionWS()
  n15 = WorldPosition()
  n16 = Subtract(
   A=Length(None=Subtract(A=@n14, B=@n15.XYZ)) {is=class MaterialExpressionLength}
   B=ScalarParameter() {defaultValue=18500, parameterName=Distance fADE}
  )
  n17 = ScalarParameter() {defaultValue=22500, parameterName=Beam soften fade}
  n18 = SphereMask(
   A=ComponentMask(None=TextureCoordinate()) {g=true}
   B=ScalarParameter() {defaultValue=1, parameterName=0-1 UV Mask Start}
   Radius=ScalarParameter() {defaultValue=0.34, parameterName=Param_3}
   Hardness=ScalarParameter() {defaultValue=0.5, parameterName=Param_7}
  )
  Multiply(
   A=Saturate(None=Divide(A=@n16, B=@n17))
   B=Multiply(A=Multiply(A=@n8.G, B=@n18), B=ScalarParameter() {defaultValue=2, parameterName=Param_2})
  )
  VectorParameter()
 ]
}
