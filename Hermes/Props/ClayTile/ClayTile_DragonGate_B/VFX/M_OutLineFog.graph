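# Translucent, two-sided unlit fog VFX material for Niagara mesh particles, filed under the DragonGate clay tile prop.
# Two panning samples of a tile texture, distorted by scrolling noise, are multiplied, sharpened and offset by a sine-based UV mask.
# The result blends two colors and drives opacity, with a view-angle fade that softens silhouette edges.
# Depth fade applies only on the default shading path, skipping forward and mobile.
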Material /Hermes/Props/ClayTile/ClayTile_DragonGate_B/VFX/M_OutLineFog
graph {
 properties {
  bUsedWithNiagaraMeshParticles=true
  blendMode=BLEND_Translucent
  pixelDepthOffsetMode=PDOM_Legacy
  shadingModel=MSM_Unlit
  twoSided=true
  usedShadingModels=Unlit
 }
 nodes [
  n1 = VectorParameter() {parameterName=Param_2, defaultValue=<rgba 5 4 0.07 -0.04>}
  n2 = TextureCoordinate()
  n3 = MaterialFunctionCall(Float2=@n2) {materialFunction=&BreakOutFloat2Components}
  n4 = Multiply(A=AppendVector(A=@n3.R, B=Add(A=@n3.R, B=@n3.R)), B=Constant2Vector() {g=1, r=-1})
  n5 = VectorParameter() {parameterName=Param_4, defaultValue=<rgba 2 1 0.1 -0.1>}
  n6 = Panner(Coordinate=Multiply(A=ComponentMask(None=@n5.RGB) {g=true, r=true}, B=@n2), Speed=AppendVector(A=@n5.B, B=@n5.B))
  n7 = Multiply(
   A=ComponentMask(None=TextureSample(UVs=@n6) {samplerType=SAMPLERTYPE_LinearColor, texture=&/T_FX_Tile_0138_Moon}.RGB) {g=true, r=true}
   B=Multiply(
    A=ScalarParameter() {defaultValue=0.1, parameterName=Param_3}
    B=TextureSample() {samplerType=SAMPLERTYPE_LinearColor, texture=&T_FX_Tile_0138_Moon}.B
   )
  )
  n8 = Multiply(A=ComponentMask(None=@n1.RGB) {g=true, r=true}, B=Add(A=@n4, B=@n7))
  n9 = AppendVector(A=@n1.B, B=@n1.B)
  n10 = TextureObjectParameter() {parameterName=Param, samplerType=SAMPLERTYPE_LinearColor, texture=&/Tile_output_8}
  n11 = VectorParameter() {parameterName=Param_3, defaultValue=<rgba 4 3 0.07 0.04>}
  n12 = Multiply(A=ComponentMask(None=@n11.RGB) {g=true, r=true}, B=Add(A=@n4, B=@n7))
  n13 = AppendVector(A=@n11.B, B=@n11.B)
  n14 = Multiply(
   A=TextureSample(UVs=Panner(Coordinate=@n8, Speed=@n9), Tex=@n10) {samplerType=SAMPLERTYPE_LinearColor}.G
   B=TextureSample(UVs=Panner(Coordinate=@n12, Speed=@n13), Tex=@n10) {samplerType=SAMPLERTYPE_LinearColor}.G
  )
  n15 = ScalarParameter() {defaultValue=1.6276}
  n16 = TextureCoordinate()
  n17 = TextureCoordinate() {coordinateIndex=1}
  n18 = ScalarParameter() {parameterName=CubeMap_Switch}
  n19 = Sine(None=Multiply(A=LinearInterpolate(A=@n16, B=@n17, Alpha=@n18)) {constB=0.5})
  n20 = MaterialFunctionCall(Float2=@n19) {materialFunction=&BreakOutFloat2Components}
  n21 = NamedRerouteDeclaration(None=MaterialFunctionCall(A=@n20.R, B=Multiply(A=@n20.R, B=@n20.R), C=@n20.R, Alpha=ScalarParameter() {defaultValue=1, parameterName=CubeMap_Select}) {materialFunction=&Lerp_3Color}.Result) {name=CubeMap, nodeColor=<rgba 1 1 1 1>}
  n22 = ScalarParameter() {defaultValue=2.621, parameterName=Param_6}
  n23 = Subtract(
   A=Power(Base=Multiply(A=@n14, B=@n15), Exp=ScalarParameter() {defaultValue=1.6292, parameterName=Param_1})
   B=Subtract(A=ScalarParameter() {defaultValue=1.7006, parameterName=Param_5}, B=Multiply(A=@n21, B=@n22) {constB=2}) {constA=2}
  )
  Multiply(
   A=LinearInterpolate(
    A=VectorParameter() {parameterName=Param_1, defaultValue=<rgba 1 0.0264 0 1>}.RGB
    B=VectorParameter() {defaultValue=<rgba 1 0 0 0>}.RGB
    Alpha=@n23
   )
   B=@n23
  )
  n24 = DepthFade(Opacity=@n23, FadeDistance=ScalarParameter() {defaultValue=100, parameterName=Param_4})
  n25 = Abs(None=DotProduct(A=VertexNormalWS(), B=CameraVectorWS()))
  Multiply(
   A=Multiply(A=ShadingPathSwitch(Default=@n24, Forward=@n23, Mobile=@n23), B=ScalarParameter() {defaultValue=0.12, parameterName=MasterOpacity})
   B=LinearInterpolate(A=Constant() {r=1}, B=Multiply(A=@n25), Alpha=ScalarParameter() {defaultValue=1, parameterName=Fresnel})
  )
  SquareRoot(None=@n19)
 ]
}
