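# M_dot is a translucent material for Niagara sprites that forms a soft round dot, stored with the ClayTile DragonGate B prop's VFX.
# A zero-hardness SphereMask over the UVs, centered at the middle with a half-width radius, creates a procedural radial falloff without textures.
# That mask multiplies particle color and particle alpha separately, so the particle system controls tint and fade.
# An unnamed vector parameter is present but unconnected.
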Material /Hermes/Props/ClayTile/ClayTile_DragonGate_B/VFX/M_dot
graph {
 properties {
  bUsedWithNiagaraSprites=true
  blendMode=BLEND_Translucent
  opacityMaskClipValue=0.01
  pixelDepthOffsetMode=PDOM_Legacy
  usedShadingModels=Default Lit
 }
 nodes [
  n1 = ParticleColor()
  n2 = Constant() {r=0.5}
  n3 = SphereMask(A=TextureCoordinate(), B=@n2, Radius=@n2) {hardnessPercent=0}
  Multiply(A=@n1.A, B=@n3)
  Multiply(A=@n1.RGB, B=@n3)
  VectorParameter()
 ]
}
