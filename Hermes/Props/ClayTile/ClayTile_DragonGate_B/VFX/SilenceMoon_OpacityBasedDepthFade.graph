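# Material function that computes an opacity-dependent depth fade, softening where translucent surfaces intersect scene geometry.
# InAlpha blends between two fade distances, defaulting to parameters OpacityBasedFade_DistanceA and B, and the result drives DepthFade.
# With those defaults, more opaque areas get a shorter, harder fade than faint ones.
# It outputs FinalAlpha, the input alpha multiplied by the fade, and the raw FadeValue.
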MaterialFunction /Hermes/Props/ClayTile/ClayTile_DragonGate_B/VFX/SilenceMoon_OpacityBasedDepthFade
graph {
 properties {bExposeToLibrary=true}
 nodes [
  n1 = FunctionInput() {inputName=InAlpha, inputType=FunctionInput_Scalar}
  n2 = ScalarParameter() {defaultValue=50, parameterName=OpacityBasedFade_DistanceA}
  n3 = ScalarParameter() {defaultValue=10, parameterName=OpacityBasedFade_DistanceB}
  n4 = DepthFade(FadeDistance=LinearInterpolate(
    A=FunctionInput(Preview=@n2) {bUsePreviewValueAsDefault=true, inputName=OpacityBasedFade_DistanceA, inputType=FunctionInput_Scalar, sortPriority=1}
    B=FunctionInput(Preview=@n3) {bUsePreviewValueAsDefault=true, inputName=OpacityBasedFade_DistanceB, inputType=FunctionInput_Scalar, sortPriority=2}
    Alpha=@n1
   ))
  FunctionOutput(None=Multiply(A=@n1, B=@n4)) {outputName=FinalAlpha}
  FunctionOutput(None=@n4) {outputName=FadeValue, sortPriority=1}
 ]
}
