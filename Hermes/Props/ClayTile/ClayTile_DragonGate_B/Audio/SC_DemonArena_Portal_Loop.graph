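# Sound cue for the looping ambient sound of the Demon Arena portal on the ClayTile Dragon Gate prop.
# A mixer layers a looping, pitch-modulated base portal wave with four endlessly looping random pickers of sparkle one-shots from sets A and B.
# Each set plays twice, with the second copy offset by a randomized start delay outside its loop and some layers mixed quieter.
# It uses attenuation, limited concurrency and a pre-effect send to the outdoor ambient sidechain bus.
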SoundCue /Hermes/Props/ClayTile/ClayTile_DragonGate_B/Audio/SC_DemonArena_Portal_Loop
root=Mixer_0
nodes [
 {id=Delay_0, children [Looping_2], properties {delayMax=2, delayMin=1.5}}
 {id=Delay_3, children [Looping_3], properties {delayMax=2.5, delayMin=2}}
 {id=Looping_0, children [Random_4]}
 {id=Looping_1, children [Random_6]}
 {id=Looping_2, children [Random_7]}
 {id=Looping_3, children [Random_5]}
 {id=Mixer_0, children [Looping_0,Delay_3,Looping_1,Delay_0,Modulator_0], properties {inputVolume [1,0.75,0.75,0.5,1]}}
 {id=Modulator_0, children [WavePlayer_4], properties {pitchMax=1.1, pitchMin=0.8}}
 {id=Random_4, children [WavePlayer_19, WavePlayer_20, WavePlayer_21, WavePlayer_22], properties {weights [1,1,1,1]}}
 {id=Random_5, children [WavePlayer_28, WavePlayer_29, WavePlayer_30, WavePlayer_31], properties {weights [1,1,1,1]}}
 {id=Random_6, children [WavePlayer_23, WavePlayer_24, WavePlayer_25, WavePlayer_26, WavePlayer_27], properties {weights [1,1,1,1,1]}}
 {id=Random_7, children [WavePlayer_32, WavePlayer_33, WavePlayer_34, WavePlayer_35, WavePlayer_36], properties {weights [1,1,1,1,1]}}
 {
  id=WavePlayer_19
  properties {soundWaveAssetPtr=&/GiftedTraveler_Montage_PortalSparkle_A_01}
 }
 {
  id=WavePlayer_20
  properties {soundWaveAssetPtr=&/GiftedTraveler_Montage_PortalSparkle_A_02}
 }
 {
  id=WavePlayer_21
  properties {soundWaveAssetPtr=&/GiftedTraveler_Montage_PortalSparkle_A_03}
 }
 {
  id=WavePlayer_22
  properties {soundWaveAssetPtr=&/GiftedTraveler_Montage_PortalSparkle_A_04}
 }
 {
  id=WavePlayer_23
  properties {soundWaveAssetPtr=&/GiftedTraveler_Montage_PortalSparkle_B_01}
 }
 {
  id=WavePlayer_24
  properties {soundWaveAssetPtr=&/GiftedTraveler_Montage_PortalSparkle_B_02}
 }
 {
  id=WavePlayer_25
  properties {soundWaveAssetPtr=&/GiftedTraveler_Montage_PortalSparkle_B_03}
 }
 {
  id=WavePlayer_26
  properties {soundWaveAssetPtr=&/GiftedTraveler_Montage_PortalSparkle_B_04}
 }
 {
  id=WavePlayer_27
  properties {soundWaveAssetPtr=&/GiftedTraveler_Montage_PortalSparkle_B_05}
 }
 {
  id=WavePlayer_28
  properties {soundWaveAssetPtr=&GiftedTraveler_Montage_PortalSparkle_A_01}
 }
 {
  id=WavePlayer_29
  properties {soundWaveAssetPtr=&GiftedTraveler_Montage_PortalSparkle_A_02}
 }
 {
  id=WavePlayer_30
  properties {soundWaveAssetPtr=&GiftedTraveler_Montage_PortalSparkle_A_03}
 }
 {
  id=WavePlayer_31
  properties {soundWaveAssetPtr=&GiftedTraveler_Montage_PortalSparkle_A_04}
 }
 {
  id=WavePlayer_32
  properties {soundWaveAssetPtr=&GiftedTraveler_Montage_PortalSparkle_B_01}
 }
 {
  id=WavePlayer_33
  properties {soundWaveAssetPtr=&GiftedTraveler_Montage_PortalSparkle_B_02}
 }
 {
  id=WavePlayer_34
  properties {soundWaveAssetPtr=&GiftedTraveler_Montage_PortalSparkle_B_03}
 }
 {
  id=WavePlayer_35
  properties {soundWaveAssetPtr=&GiftedTraveler_Montage_PortalSparkle_B_04}
 }
 {
  id=WavePlayer_36
  properties {soundWaveAssetPtr=&GiftedTraveler_Montage_PortalSparkle_B_05}
 }
 {id=WavePlayer_4, properties {bLooping=true, soundWaveAssetPtr=&/DemonArena_Portal_Loop_01}}
]
properties {
 attenuationSettings=&/AS_DemonArena_Portal_Loop
 bOverrideConcurrency=true
 maxDistance=4300
 soundClassObject=&Fort_Ambient_Props_Athena
 volumeMultiplier=1
 concurrencyOverrides {maxCount=3}
 preEffectBusSends [{audioBus=/Game/Sounds/Modulation/EnvelopeFollowers/Volume/AB_SideChain_FortAmbient_Outdoors_Vol, soundSourceBus=null}]
}
